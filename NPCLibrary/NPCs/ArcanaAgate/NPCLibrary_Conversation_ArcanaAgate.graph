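# Conversation database for the ArcanaAgate NPC, a Scooby-Doo-themed character who greets players with investigation banter.
# Its entry point feeds one speech node that falls back to a default line.
# Contextual lines cover the first-ever conversation, repeat talks within a match and specific player outfits, with curve-driven priorities for first-meeting and outfit reactions.
# The speech node branches through link nodes to shared NPC quests, basic services and a Season 37 voice fork.
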ConversationDatabase /NPCLibrary/NPCs/ArcanaAgate/NPCLibrary_Conversation_ArcanaAgate
entry_points [{tag=Conversation.Entry.ArcanaAgate, destinations [#1]}]
nodes [
 ConversationEntryPointNode_7 {
  id=#1
  class=ConversationEntryPointNode
  outputs [#2]
  properties {entryTag {tagName=Conversation.Entry.ArcanaAgate}}
 }
 Speech_C_4 {
  id=#2
  class=Speech_C
  outputs [#3, #4, #5]
  properties {
   nodeName=""
   generalConfig {
    defaultMessage="I can defend myself. I've got a black belt, you know!"
    contextualMessages [
     {
      message="I'm out here investigating this Dark Presence! Got any leads?"
      contextRequirements [{requirement {class=FortControllerRequirement_IsFirstEverConversationWithNPC}}]
      priority {curve {rowName=Tandem.Speech.Priority.FirstEverInteraction}}
     }
     {
      message="Jeepers! This Island sure is creepy."
      contextRequirements [{requirement {
         class=FortControllerRequirement_CheckNPCMatchConversationCount
         properties {nPCUniqueIdTag {tagName=AISpawnerData.Type.Tandem.ArcanaAgate}, requiredConversationCountConfig {requirement=GreaterThan}}
        }}]
      priority {curve {curveTable=null, rowName=None}}
     }
     {
      message=Fred said we should all split up to cover more ground for the investigation.
      contextRequirements [{requirement {
         class=FortControllerRequirement_CheckNPCMatchConversationCount
         properties {nPCUniqueIdTag {tagName=AISpawnerData.Type.Tandem.ArcanaAgate}, requiredConversationCountConfig {requirement=GreaterThan}}
        }}]
      priority {curve {curveTable=null, rowName=None}}
     }
     {
      message="I can defend myself. I've got a black belt, you know!"
      contextRequirements [{requirement {
         class=FortControllerRequirement_CheckNPCMatchConversationCount
         properties {nPCUniqueIdTag {tagName=AISpawnerData.Type.Tandem.ArcanaAgate}, requiredConversationCountConfig {requirement=GreaterThan}}
        }}]
      priority {curve {curveTable=null, rowName=None}}
     }
     {
      message=You've got Scooby Snack crumbs on your chin.
      contextRequirements [{requirement {class=FortControllerRequirement_HasCID, properties {allowedCIds [/BRCosmetics/Athena/Items/Cosmetics/Characters/Character_LycheeNickel]}}}]
      priority {curve {rowName=Tandem.Speech.Priority.OutfitReaction}}
     }
     {
      message="Find any clues, gang?"
      contextRequirements [{requirement {
         class=FortControllerRequirement_HasCID
         properties {allowedCIds [/BRCosmetics/Athena/Items/Cosmetics/Characters/Character_BaskIsle, /BRCosmetics/Athena/Items/Cosmetics/Characters/Character_FluteLamp]}
        }}]
      priority {curve {rowName=Tandem.Speech.Priority.OutfitReaction}}
     }
     {
      message="Before you ask, your backup ascot is in the van."
      contextRequirements [{requirement {class=FortControllerRequirement_HasCID, properties {allowedCIds [/BRCosmetics/Athena/Items/Cosmetics/Characters/Character_OboeThorn]}}}]
      priority {curve {rowName=Tandem.Speech.Priority.OutfitReaction}}
     }
    ]
   }
  }
 }
 ConversationLinkNode_2 {
  id=#5
  class=ConversationLinkNode
  properties {remoteEntryTag {tagName=Conversation.Entry.Season37.VoiceFork.ArcanaAgate}}
 }
 ConversationLinkNode_0 {id=#4, class=ConversationLinkNode, properties {remoteEntryTag {tagName=Conversation.Entry.BasicServices}}}
 ConversationLinkNode_1 {id=#3, class=ConversationLinkNode, properties {remoteEntryTag {tagName=Conversation.Entry.NPCLibrary.Quests}}}
]
